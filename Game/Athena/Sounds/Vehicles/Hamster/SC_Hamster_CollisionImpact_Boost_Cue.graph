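# Sound cue for the boosted collision impact of the Hamster vehicle, in the vehicles sound class.
# A mixer layers three modulated branches at different volumes: a random boost impact, a quieter random normal impact and a random HamsterV2 hard impact.
# The hard-impact modulator has its own raised pitch ceiling, and the whole cue is pitched up.
# It borrows the golf cart close-movement attenuation asset.
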SoundCue /Game/Athena/Sounds/Vehicles/Hamster/SC_Hamster_CollisionImpact_Boost_Cue
root=Mixer_0
nodes [
 {id=Mixer_0, children [Modulator_0,Modulator_1,Modulator_2], properties {inputVolume [1.25,0.5,1]}}
 {id=Modulator_0, children [Random_0]}
 {id=Modulator_1, children [Random_1]}
 {id=Modulator_2, children [Random_2], properties {pitchMax=1.2}}
 {id=Random_0, children [WavePlayer_0,WavePlayer_1,WavePlayer_5], properties {weights [1,1,1]}}
 {id=Random_1, children [WavePlayer_2,WavePlayer_3], properties {weights [1,1]}}
 {id=Random_2, children [WavePlayer_4,WavePlayer_6,WavePlayer_7], properties {weights [1,1,1]}}
 {id=WavePlayer_0, properties {soundWaveAssetPtr=&/Hamster_Impact_Boost_01}}
 {id=WavePlayer_1, properties {soundWaveAssetPtr=&/Hamster_Impact_Boost_02}}
 {id=WavePlayer_2, properties {soundWaveAssetPtr=&/Hamster_Impact_Normal_01}}
 {id=WavePlayer_3, properties {soundWaveAssetPtr=&/Hamster_Impact_Normal_02}}
 {id=WavePlayer_4, properties {soundWaveAssetPtr=&/HamsterV2_Impact_Hard_01}}
 {id=WavePlayer_5, properties {soundWaveAssetPtr=&/Hamster_Impact_Boost_03}}
 {id=WavePlayer_6, properties {soundWaveAssetPtr=&/HamsterV2_Impact_Hard_02}}
 {id=WavePlayer_7, properties {soundWaveAssetPtr=&/HamsterV2_Impact_Hard_03}}
]
properties {
 attenuationSettings=/Game/Sounds/Attenuation/Fort_GolfCart_Movement_Close_Att
 maxDistance=5300
 pitchMultiplier=1.5
 soundClassObject=&Fort_Vehicles
 volumeMultiplier=1
}
