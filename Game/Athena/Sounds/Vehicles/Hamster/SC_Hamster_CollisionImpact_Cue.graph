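# Sound cue for the Hamster vehicle's collision impact, played in the vehicles sound class.
# A mixer layers one of two random normal impact waves with one of three random HamsterV2 hard impact waves, weighting the hard layer louder.
# Modulators randomize pitch on both layers, pushing the normal layer well above natural pitch and also varying its volume.
# Spatialization borrows the golf cart close-movement attenuation asset.
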SoundCue /Game/Athena/Sounds/Vehicles/Hamster/SC_Hamster_CollisionImpact_Cue
root=Mixer_0
nodes [
 {id=Mixer_0, children [Modulator_0,Modulator_1], properties {inputVolume [1,1.5]}}
 {id=Modulator_0, children [Random_0], properties {pitchMax=1.6, pitchMin=1.5, volumeMin=0.75}}
 {id=Modulator_1, children [Random_1], properties {pitchMax=1.2, pitchMin=1, volumeMax=1, volumeMin=1}}
 {id=Random_0, children [WavePlayer_0,WavePlayer_1], properties {weights [1,1]}}
 {id=Random_1, children [WavePlayer_2,WavePlayer_3,WavePlayer_4], properties {weights [1,1,1]}}
 {id=WavePlayer_0, properties {soundWaveAssetPtr=&/Hamster_Impact_Normal_01}}
 {id=WavePlayer_1, properties {soundWaveAssetPtr=&/Hamster_Impact_Normal_02}}
 {id=WavePlayer_2, properties {soundWaveAssetPtr=&/HamsterV2_Impact_Hard_01}}
 {id=WavePlayer_3, properties {soundWaveAssetPtr=&/HamsterV2_Impact_Hard_02}}
 {id=WavePlayer_4, properties {soundWaveAssetPtr=&/HamsterV2_Impact_Hard_03}}
]
properties {
 attenuationSettings=/Game/Sounds/Attenuation/Fort_GolfCart_Movement_Close_Att
 maxDistance=5300
 soundClassObject=&Fort_Vehicles
 volumeMultiplier=1
}
