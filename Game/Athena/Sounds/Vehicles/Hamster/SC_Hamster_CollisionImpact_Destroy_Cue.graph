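# Sound cue for the Hamster vehicle's destroy collision impact.
# It picks one of two destroy impact waves at equal weight, passes it through a modulator at default settings, and starts it after a short fixed delay.
# The cue raises pitch and volume, reuses the golf cart distant movement attenuation, and plays in the medium explosives sound class.
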SoundCue /Game/Athena/Sounds/Vehicles/Hamster/SC_Hamster_CollisionImpact_Destroy_Cue
root=Delay_0
nodes [
 {id=Delay_0, children [Modulator_0], properties {delayMax=0.15, delayMin=0.15}}
 {id=Modulator_0, children [Random_1]}
 {id=Random_1, children [WavePlayer_2,WavePlayer_3], properties {weights [1,1]}}
 {id=WavePlayer_2, properties {soundWaveAssetPtr=&/Hamster_Impact_Destroy_01}}
 {id=WavePlayer_3, properties {soundWaveAssetPtr=&/Hamster_Impact_Destroy_02}}
]
properties {
 attenuationSettings=/Game/Sounds/Attenuation/Fort_GolfCart_Movement_Distant_Att
 maxDistance=10100
 pitchMultiplier=1.35
 soundClassObject=/Game/Sounds/FortSoundClassesAndModes/Fort_Explosives_Medium_PSM
 volumeMultiplier=1.75
}
