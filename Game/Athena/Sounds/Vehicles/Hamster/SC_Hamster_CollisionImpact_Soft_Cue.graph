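# Sound cue for a soft collision impact on the Hamster vehicle in Athena.
# It mixes a random soft impact wave with a random hard impact wave at half input level, each layer raised in pitch by its own modulator.
# An outer modulator varies the mix's volume, and the cue further raises pitch and lowers volume.
# It plays in the vehicles sound class with borrowed golf cart close-movement attenuation.
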SoundCue /Game/Athena/Sounds/Vehicles/Hamster/SC_Hamster_CollisionImpact_Soft_Cue
root=Modulator_2
nodes [
 {id=Mixer_0, children [Modulator_0,Modulator_1], properties {inputVolume [1,0.5]}}
 {id=Modulator_0, children [Random_1], properties {pitchMax=1.6, pitchMin=1.5, volumeMin=0.5}}
 {id=Modulator_1, children [Random_0], properties {pitchMax=1.6, pitchMin=1.5}}
 {id=Modulator_2, children [Mixer_0], properties {volumeMin=0.5}}
 {id=Random_0, children [WavePlayer_0,WavePlayer_1,WavePlayer_5], properties {weights [1,1,1]}}
 {id=Random_1, children [WavePlayer_2,WavePlayer_3,WavePlayer_4], properties {weights [1,1,1]}}
 {id=WavePlayer_0, properties {soundWaveAssetPtr=&/HamsterV2_Impact_Hard_01}}
 {id=WavePlayer_1, properties {soundWaveAssetPtr=&/HamsterV2_Impact_Hard_02}}
 {id=WavePlayer_2, properties {soundWaveAssetPtr=&/Hamster_Impact_Soft_02}}
 {id=WavePlayer_3, properties {soundWaveAssetPtr=&/Hamster_Impact_Soft_03}}
 {id=WavePlayer_4, properties {soundWaveAssetPtr=&/Hamster_Impact_Soft_01}}
 {id=WavePlayer_5, properties {soundWaveAssetPtr=&/HamsterV2_Impact_Hard_03}}
]
properties {
 attenuationSettings=/Game/Sounds/Attenuation/Fort_GolfCart_Movement_Close_Att
 maxDistance=5300
 pitchMultiplier=2
 soundClassObject=&Fort_Vehicles
 volumeMultiplier=0.5
}
