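# Sound cue that, judging by its name, plays when the Hamster vehicle reaches a damage threshold.
# A random node picks one of two impact-break waves with equal weight and feeds a modulator that slightly varies pitch.
# It uses the Fort_Vehicles sound class and borrows a close-range golf cart movement attenuation setting.
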SoundCue /Game/Athena/Sounds/Vehicles/Hamster/SC_Hamster_DamageThreshold_Reached_Cue
root=Modulator_0
nodes [
 {id=Modulator_0, children [Random_0], properties {pitchMax=1.1, pitchMin=0.9}}
 {id=Random_0, children [WavePlayer_0,WavePlayer_1], properties {weights [1,1]}}
 {id=WavePlayer_0, properties {soundWaveAssetPtr=&/Hamster_Impact_Break_01}}
 {id=WavePlayer_1, properties {soundWaveAssetPtr=&/Hamster_Impact_Break_02}}
]
properties {
 attenuationSettings=/Game/Sounds/Attenuation/Fort_GolfCart_Movement_Close_Att
 maxDistance=5300
 soundClassObject=&Fort_Vehicles
 volumeMultiplier=1
}
