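# Sound cue for the Hamster vehicle's energy-depleted alert.
# A mixer layers the main depleted wave with a quieter alarm wave that starts after a short fixed delay.
# It uses the shared vehicle prop attenuation and the vehicle interior sound class, with a boosted cue volume.
# A concurrency override blocks new instances instead of cutting off ones already playing.
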SoundCue /Game/Athena/Sounds/Vehicles/Hamster/SC_Hamster_Energy_Depleted_Cue
root=Mixer_0
nodes [
 {id=Delay_0, children [WavePlayer_0], properties {delayMax=0.2, delayMin=0.2}}
 {id=Mixer_0, children [WavePlayer_2,Delay_0], properties {inputVolume [1,0.6]}}
 {id=WavePlayer_0, properties {soundWaveAssetPtr=&/HamsterV2_Energy_Depleted_Alarm}}
 {id=WavePlayer_2, properties {soundWaveAssetPtr=&/HamsterV2_Energy_Depleted}}
]
properties {
 attenuationSettings=/Game/Sounds/Attenuation/Fort_Prop_Vehicles_Att
 bOverrideConcurrency=true
 maxDistance=3000
 soundClassObject=&Fort_Vehicles_Interior
 volumeMultiplier=1.7
 concurrencyOverrides {resolutionRule=PreventNew}
}
