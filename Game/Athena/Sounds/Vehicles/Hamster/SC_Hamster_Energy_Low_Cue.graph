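# Sound cue for the Hamster vehicle's low-energy start sound, played in the vehicle interior sound class.
# A single wave player feeds the HamsterV2 low-energy start wave straight to the output at a slightly raised volume.
# It uses the shared vehicle prop attenuation and overrides concurrency so new instances are blocked rather than interrupting those already playing.
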SoundCue /Game/Athena/Sounds/Vehicles/Hamster/SC_Hamster_Energy_Low_Cue
root=WavePlayer_2
nodes [{id=WavePlayer_2, properties {soundWaveAssetPtr=&/HamsterV2_Energy_Low_Start}}]
properties {
 attenuationSettings=/Game/Sounds/Attenuation/Fort_Prop_Vehicles_Att
 bOverrideConcurrency=true
 maxDistance=3000
 soundClassObject=&Fort_Vehicles_Interior
 volumeMultiplier=1.05
 concurrencyOverrides {resolutionRule=PreventNew}
}
