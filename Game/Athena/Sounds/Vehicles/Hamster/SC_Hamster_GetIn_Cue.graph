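# Sound cue for getting into the Hamster vehicle, in the vehicle sound class with borrowed close-range golf cart attenuation.
# A quieter get-in sample plays immediately, mixed with a briefly delayed layer of a random soft impact and a boosted servo sound.
# The servo gets a further random delay and a modulator, and its quality-level switch offers four variants, two variants, or empty slots for the last two levels.
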SoundCue /Game/Athena/Sounds/Vehicles/Hamster/SC_Hamster_GetIn_Cue
root=Mixer_1
nodes [
 {id=Delay_0, children [Modulator_0], properties {delayMax=0.4, delayMin=0.2}}
 {id=Delay_1, children [Mixer_0], properties {delayMax=0.2, delayMin=0.2}}
 {id=Mixer_0, children [Random_0,Delay_0], properties {inputVolume [1,1.5]}}
 {id=Mixer_1, children [Delay_1,WavePlayer_4], properties {inputVolume [1,0.35]}}
 {id=Modulator_0, children [QualityLevel_0]}
 {id=QualityLevel_0, children [Random_2,Random_3,null,null]}
 {id=Random_0, children [WavePlayer_1,WavePlayer_2], properties {weights [1,1]}}
 {id=Random_2, children [WavePlayer_0, WavePlayer_8, WavePlayer_9, WavePlayer_10], properties {weights [1,1,1,1]}}
 {id=Random_3, children [WavePlayer_11, WavePlayer_12], properties {weights [1,1]}}
 {id=WavePlayer_0, properties {soundWaveAssetPtr=&/Hamster_Servo_Small_03}}
 {id=WavePlayer_1, properties {soundWaveAssetPtr=&/Hamster_Impact_Soft_03}}
 {id=WavePlayer_10, properties {soundWaveAssetPtr=&/Hamster_Servo_Small_02}}
 {id=WavePlayer_11, properties {soundWaveAssetPtr=&Hamster_Servo_Small_03}}
 {id=WavePlayer_12, properties {soundWaveAssetPtr=&/Hamster_Servo_Small_04}}
 {id=WavePlayer_2, properties {soundWaveAssetPtr=&/Hamster_Impact_Soft_01}}
 {id=WavePlayer_4, properties {soundWaveAssetPtr=&/HamsterV2_Get_In_01}}
 {id=WavePlayer_8, properties {soundWaveAssetPtr=&Hamster_Servo_Small_04}}
 {id=WavePlayer_9, properties {soundWaveAssetPtr=&/Hamster_Servo_Small_01}}
]
properties {
 attenuationSettings=/Game/Sounds/Attenuation/Fort_GolfCart_Movement_Close_Att
 maxDistance=5300
 soundClassObject=&Fort_Vehicles
 volumeMultiplier=1
}
