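# Sound cue for getting out of the Hamster vehicle, routed to the vehicles sound class with a borrowed close-range golf cart attenuation.
# A top mixer layers a quieter get-out wave with a briefly delayed submix of a random soft impact and a louder servo.
# The servo follows a further randomized delay and a modulator, with a quality-level switch choosing a four-wave or two-wave random pool and leaving its last two slots empty.
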SoundCue /Game/Athena/Sounds/Vehicles/Hamster/SC_Hamster_GetOut_Cue
root=Mixer_2
nodes [
 {id=Delay_1, children [Modulator_1], properties {delayMax=0.4, delayMin=0.2}}
 {id=Delay_2, children [Mixer_1], properties {delayMax=0.2, delayMin=0.2}}
 {id=Mixer_1, children [Random_1,Delay_1], properties {inputVolume [1,1.5]}}
 {id=Mixer_2, children [Delay_2, WavePlayer_18], properties {inputVolume [1,0.35]}}
 {id=Modulator_1, children [QualityLevel_1]}
 {id=QualityLevel_1, children [Random_4,Random_5,null,null]}
 {id=Random_1, children [WavePlayer_4,WavePlayer_5], properties {weights [1,1]}}
 {id=Random_4, children [WavePlayer_6, WavePlayer_7, WavePlayer_13, WavePlayer_14], properties {weights [1,1,1,1]}}
 {id=Random_5, children [WavePlayer_15, WavePlayer_16], properties {weights [1,1]}}
 {id=WavePlayer_13, properties {soundWaveAssetPtr=&/Hamster_Servo_Small_01}}
 {id=WavePlayer_14, properties {soundWaveAssetPtr=&/Hamster_Servo_Small_02}}
 {id=WavePlayer_15, properties {soundWaveAssetPtr=&/Hamster_Servo_Small_03}}
 {id=WavePlayer_16, properties {soundWaveAssetPtr=&/Hamster_Servo_Small_04}}
 {id=WavePlayer_18, properties {soundWaveAssetPtr=&/HamsterV2_Get_Out_01}}
 {id=WavePlayer_4, properties {soundWaveAssetPtr=&/Hamster_Impact_Soft_03}}
 {id=WavePlayer_5, properties {soundWaveAssetPtr=&/Hamster_Impact_Soft_01}}
 {id=WavePlayer_6, properties {soundWaveAssetPtr=&Hamster_Servo_Small_03}}
 {id=WavePlayer_7, properties {soundWaveAssetPtr=&Hamster_Servo_Small_04}}
]
properties {
 attenuationSettings=/Game/Sounds/Attenuation/Fort_GolfCart_Movement_Close_Att
 maxDistance=5300
 soundClassObject=&Fort_Vehicles
 volumeMultiplier=1
}
