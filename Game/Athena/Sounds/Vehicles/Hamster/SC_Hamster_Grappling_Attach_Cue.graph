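# Sound cue for the Hamster vehicle's grappling attach sound.
# A Random node picks one of two attach waves with equal weight, feeding a Mixer with an empty second input, then a Modulator that slightly varies volume.
# It plays in the impacts sound class, borrowing golf cart close-movement attenuation and rifle bullet impact concurrency rather than assets of its own.
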SoundCue /Game/Athena/Sounds/Vehicles/Hamster/SC_Hamster_Grappling_Attach_Cue
root=Modulator_0
nodes [
 {id=Mixer_1, children [Random_2,null], properties {inputVolume [1.5,1]}}
 {id=Modulator_0, children [Mixer_1], properties {volumeMax=1.2, volumeMin=1.1}}
 {id=Random_2, children [WavePlayer_6,WavePlayer_7], properties {weights [1,1]}}
 {id=WavePlayer_6, properties {soundWaveAssetPtr=&/Hamster_Grappling_Attach_02}}
 {id=WavePlayer_7, properties {soundWaveAssetPtr=&/Hamster_Grappling_Attach_01}}
]
properties {
 attenuationSettings=/Game/Sounds/Attenuation/Fort_GolfCart_Movement_Close_Att
 maxDistance=5300
 soundClassObject=&Fort_Impacts
 volumeMultiplier=1.9
 attenuationOverrides {bEnableReverbSend=false, falloffDistance=3000, nonSpatializedRadiusStart=1000, stereoSpread=2000, attenuationShapeExtents {x=2500}}
 concurrencySet [&RifleBulletImpact_Conc]
}
