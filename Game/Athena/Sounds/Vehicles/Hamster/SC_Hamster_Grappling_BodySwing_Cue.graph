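# Sound cue for the Hamster vehicle's grappling body swing, reusing Grappling Hook weapon waves.
# A Mixer layers an immediate, louder random body-pull wave with a quieter random body-swing wave that follows after a short fixed delay.
# Each layer picks from four equally weighted waves through its own Modulator for variation.
# It plays in the vehicles sound class with golf cart movement attenuation and player melee weapon concurrency.
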SoundCue /Game/Athena/Sounds/Vehicles/Hamster/SC_Hamster_Grappling_BodySwing_Cue
root=Mixer_0
nodes [
 {id=Delay_0, children [Modulator_0], properties {delayMax=0.3, delayMin=0.3}}
 {id=Mixer_0, children [Delay_0,Modulator_1], properties {inputVolume [0.6,1.4]}}
 {id=Modulator_0, children [Random_0], properties {pitchMax=1.1, pitchMin=0.9}}
 {id=Modulator_1, children [Random_1]}
 {id=Random_0, children [WavePlayer_0,WavePlayer_1,WavePlayer_2,WavePlayer_3], properties {weights [1,1,1,1]}}
 {id=Random_1, children [WavePlayer_8, WavePlayer_9, WavePlayer_10, WavePlayer_11], properties {weights [1,1,1,1]}}
 {id=WavePlayer_0, properties {soundWaveAssetPtr=/Game/Sounds/Fort_Weapons/GrapplingHook/Grappling_BodySwing_01}}
 {id=WavePlayer_1, properties {soundWaveAssetPtr=/Game/Sounds/Fort_Weapons/GrapplingHook/Grappling_BodySwing_02}}
 {id=WavePlayer_10, properties {soundWaveAssetPtr=/Game/Sounds/Fort_Weapons/GrapplingHook/GrapplingHook_BodyPull_Layer03}}
 {id=WavePlayer_11, properties {soundWaveAssetPtr=/Game/Sounds/Fort_Weapons/GrapplingHook/GrapplingHook_BodyPull_Layer04}}
 {id=WavePlayer_2, properties {soundWaveAssetPtr=/Game/Sounds/Fort_Weapons/GrapplingHook/Grappling_BodySwing_03}}
 {id=WavePlayer_3, properties {soundWaveAssetPtr=/Game/Sounds/Fort_Weapons/GrapplingHook/Grappling_BodySwing_04}}
 {id=WavePlayer_8, properties {soundWaveAssetPtr=/Game/Sounds/Fort_Weapons/GrapplingHook/GrapplingHook_BodyPull_Layer01}}
 {id=WavePlayer_9, properties {soundWaveAssetPtr=/Game/Sounds/Fort_Weapons/GrapplingHook/GrapplingHook_BodyPull_Layer02}}
]
properties {
 attenuationSettings=/Game/Sounds/Attenuation/Fort_GolfCart_Movement_Close_Att
 maxDistance=5300
 soundClassObject=&Fort_Vehicles
 volumeMultiplier=1.2
 attenuationOverrides {bEnableReverbSend=false, falloffDistance=3000, nonSpatializedRadiusStart=800, stereoSpread=2000, attenuationShapeExtents {x=0}}
 concurrencySet [&Player_Melee_Weapon_Conc]
}
