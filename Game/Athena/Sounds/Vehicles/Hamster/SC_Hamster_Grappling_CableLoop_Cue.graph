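# Looping sound cue for the grappling cable on the Hamster vehicle in Athena.
# A single looping Wave Player feeds the Hamster_Grappling_CableLoop_01 wave straight to the output, with no randomization or modulation.
# It plays in the Fort_Vehicles sound class with the close golf cart movement attenuation asset, and uses the RifleBulletImpact concurrency group rather than a vehicle-specific one.
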SoundCue /Game/Athena/Sounds/Vehicles/Hamster/SC_Hamster_Grappling_CableLoop_Cue
root=WavePlayer_6
nodes [{id=WavePlayer_6, properties {bLooping=true, soundWaveAssetPtr=&/Hamster_Grappling_CableLoop_01}}]
properties {
 attenuationSettings=/Game/Sounds/Attenuation/Fort_GolfCart_Movement_Close_Att
 maxDistance=5300
 soundClassObject=&Fort_Vehicles
 volumeMultiplier=1
 attenuationOverrides {bEnableReverbSend=false, falloffDistance=3000, nonSpatializedRadiusStart=1000, stereoSpread=2000, attenuationShapeExtents {x=2500}}
 concurrencySet [&RifleBulletImpact_Conc]
}
