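# Sound cue for firing the grappling hook on the Hamster vehicle, mixing three layers with their own attenuation and sound class overrides.
# Randomized close and medium shots reused from the GrapplingHook weapon sit at reduced volume under a full-volume Hamster-specific shot layer.
# A quality-level switch gives the Hamster layer three random variations, two, or a single wave.
# The cue slightly lowers pitch and overrides concurrency to prevent new sounds at the limit.
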SoundCue /Game/Athena/Sounds/Vehicles/Hamster/SC_Hamster_Grappling_Fire_Cue
root=Mixer_1
nodes [
 {id=Attenuation_3, children [Random_1], properties {attenuationSettings=&Fort_Weapon_Projectile_Close_Athena_Att}}
 {
  id=Attenuation_4
  children [Random_2]
  properties {attenuationSettings=&Fort_Weapon_Projectile_Medium_Mono_Athena_Att}
 }
 {id=Attenuation_5, children [QualityLevel_0], properties {attenuationSettings=/Game/Sounds/Attenuation/Fort_GolfCart_Movement_Close_Att}}
 {id=Mixer_1, children [SoundClass_3,SoundClass_2,SoundClass_4], properties {inputVolume [0.3,0.3,1]}}
 {id=QualityLevel_0, children [Random_0, Random_3, WavePlayer_10, WavePlayer_10]}
 {id=Random_0, children [WavePlayer_0,WavePlayer_1,WavePlayer_7], properties {weights [1,1,1]}}
 {id=Random_1, children [WavePlayer_2,WavePlayer_3,WavePlayer_4,WavePlayer_5], properties {weights [1,1,1,1]}}
 {id=Random_2, children [WavePlayer_6, WavePlayer_8, WavePlayer_9, WavePlayer_14], properties {weights [1,1,1,1]}}
 {id=Random_3, children [WavePlayer_10, WavePlayer_11], properties {weights [1,1]}}
 {
  id=SoundClass_2
  children [Attenuation_4]
  properties {soundClassOverride=&Fort_WeaponRanged_Fire_Distant}
 }
 {
  id=SoundClass_3
  children [Attenuation_3]
  properties {soundClassOverride=&Fort_WeaponRanged_Fire_3P_PSM}
 }
 {
  id=SoundClass_4
  children [Attenuation_5]
  properties {soundClassOverride=&Fort_WeaponRanged_Fire_Distant}
 }
 {id=WavePlayer_0, properties {soundWaveAssetPtr=&/Hamster_Grappling_Shoot_01}}
 {id=WavePlayer_1, properties {soundWaveAssetPtr=&/Hamster_Grappling_Shoot_02}}
 {id=WavePlayer_10, properties {soundWaveAssetPtr=&Hamster_Grappling_Shoot_01}}
 {id=WavePlayer_11, properties {soundWaveAssetPtr=&Hamster_Grappling_Shoot_02}}
 {id=WavePlayer_14, properties {soundWaveAssetPtr=/Game/Sounds/Fort_Weapons/GrapplingHook/Grappling_Fire_Medium_04}}
 {id=WavePlayer_2, properties {soundWaveAssetPtr=/Game/Sounds/Fort_Weapons/GrapplingHook/Grappling_Fire_01}}
 {id=WavePlayer_3, properties {soundWaveAssetPtr=/Game/Sounds/Fort_Weapons/GrapplingHook/Grappling_Fire_02}}
 {id=WavePlayer_4, properties {soundWaveAssetPtr=/Game/Sounds/Fort_Weapons/GrapplingHook/Grappling_Fire_03}}
 {id=WavePlayer_5, properties {soundWaveAssetPtr=/Game/Sounds/Fort_Weapons/GrapplingHook/Grappling_Fire_04}}
 {id=WavePlayer_6, properties {soundWaveAssetPtr=/Game/Sounds/Fort_Weapons/GrapplingHook/Grappling_Fire_Medium_01}}
 {id=WavePlayer_7, properties {soundWaveAssetPtr=&/Hamster_Grappling_Shoot_03}}
 {id=WavePlayer_8, properties {soundWaveAssetPtr=/Game/Sounds/Fort_Weapons/GrapplingHook/Grappling_Fire_Medium_02}}
 {id=WavePlayer_9, properties {soundWaveAssetPtr=/Game/Sounds/Fort_Weapons/GrapplingHook/Grappling_Fire_Medium_03}}
]
properties {
 bOverrideConcurrency=true
 maxDistance=10579.2871
 pitchMultiplier=0.95
 volumeMultiplier=0.8
 attenuationOverrides {falloffDistance=5000, attenuationShapeExtents {x=1000}}
 concurrencyOverrides {resolutionRule=PreventNew}
}
