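# Sound cue for the Hamster vehicle's grappling tow hook hitting water, built from reused medium bullet water impact waves.
# A fixed short delay feeds a modulator that lowers pitch and boosts volume, then a quality switch.
# The first quality level randomizes five waves, while the other three share a two-wave random excluded from branch culling.
# It uses tow hook water attenuation, the impacts sound class and rifle bullet impact concurrency.
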SoundCue /Game/Athena/Sounds/Vehicles/Hamster/SC_Hamster_Grappling_Water_Impact_Cue
root=Delay_0
nodes [
 {id=Delay_0, children [Modulator_0], properties {delayMax=0.1, delayMin=0.1}}
 {id=Modulator_0, children [QualityLevel_1], properties {pitchMax=0.9, pitchMin=0.85, volumeMax=1.2, volumeMin=1.1}}
 {id=QualityLevel_1, children [Random_0,Random_3,Random_3,Random_3]}
 {id=Random_0, children [WavePlayer_0,WavePlayer_1,WavePlayer_2,WavePlayer_3,WavePlayer_4], properties {weights [1,1,1,1,1]}}
 {id=Random_3, children [WavePlayer_21, WavePlayer_24], properties {bShouldExcludeFromBranchCulling=true, weights [1,1]}}
 {id=WavePlayer_0, properties {soundWaveAssetPtr=/Game/Sounds/Fort_Weapons/Impacts/Bullet_Impacts_S10/Bullet_Impact_Water_Med_10}}
 {id=WavePlayer_1, properties {soundWaveAssetPtr=/Game/Sounds/Fort_Weapons/Impacts/Bullet_Impacts_S10/Bullet_Impact_Water_Med_01}}
 {id=WavePlayer_2, properties {soundWaveAssetPtr=/Game/Sounds/Fort_Weapons/Impacts/Bullet_Impacts_S10/Bullet_Impact_Water_Med_02}}
 {id=WavePlayer_21, properties {soundWaveAssetPtr=&Bullet_Impact_Water_Med_01}}
 {id=WavePlayer_24, properties {soundWaveAssetPtr=/Game/Sounds/Fort_Weapons/Impacts/Bullet_Impacts_S10/Bullet_Impact_Water_Med_03}}
 {id=WavePlayer_3, properties {soundWaveAssetPtr=&Bullet_Impact_Water_Med_03}}
 {id=WavePlayer_4, properties {soundWaveAssetPtr=/Game/Sounds/Fort_Weapons/Impacts/Bullet_Impacts_S10/Bullet_Impact_Water_Med_04}}
]
properties {
 attenuationSettings=&/Hamster_Towhook_WaterImpact_Att
 maxDistance=5300
 soundClassObject=&Fort_Impacts
 volumeMultiplier=2.5
 attenuationOverrides {bEnableReverbSend=false, falloffDistance=3000, nonSpatializedRadiusStart=1000, stereoSpread=2000, attenuationShapeExtents {x=2500}}
 concurrencySet [&RifleBulletImpact_Conc]
}
