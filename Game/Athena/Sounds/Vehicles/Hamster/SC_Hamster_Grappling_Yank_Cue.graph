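# Sound cue named for the grappling yank of the Hamster vehicle, with an empty node graph, so it references no sound wave.
# It still carries playback settings: the vehicle sound class, reduced volume and pitch, and a close-range golf cart movement attenuation asset.
# A shared concurrency rule, named for allowing two instances per owner and stopping the oldest, is assigned.
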SoundCue /Game/Athena/Sounds/Vehicles/Hamster/SC_Hamster_Grappling_Yank_Cue
root=null
nodes []
properties {
 attenuationSettings=/Game/Sounds/Attenuation/Fort_GolfCart_Movement_Close_Att
 maxDistance=5300
 pitchMultiplier=0.9
 soundClassObject=&Fort_Vehicles
 volumeMultiplier=0.45
 concurrencySet [/Game/Sounds/Concurrency_Rules/TwoMaximum_Owner_StopOldest]
}
