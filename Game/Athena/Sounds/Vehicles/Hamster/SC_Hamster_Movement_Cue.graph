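# SoundCue for the Hamster vehicle's looping movement audio, mixing surface roll, turn, idle and water layers under a smoothed Doppler root.
# Continuous parameters drive the layers: Speed, OnGround and Damage for rolling, TurnAmount for turning, and HasDriver, InvSpeed and OutOfFuel for idle.
# WaterSpeed scales borrowed Meatball wake loops, with only the calm wake gated by an IsInWater branch.
# Idle and calm-wake chains are reused through extra TurnAmount modulators, and nested GolfCart attenuations range some layers separately.
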SoundCue /Game/Athena/Sounds/Vehicles/Hamster/SC_Hamster_Movement_Cue
root=Doppler_0
nodes [
 {
  id=Attenuation_0
  children [ModulatorContinuous_9]
  properties {attenuationSettings=/Game/Sounds/Attenuation/Fort_GolfCart_Movement_Close_Att}
 }
 {id=Attenuation_1, children [Modulator_0], properties {attenuationSettings=/Game/Sounds/Attenuation/Fort_GolfCart_Movement_Medium_Att}}
 {id=Attenuation_3, children [Branch_0], properties {attenuationSettings=&Fort_GolfCart_Movement_Close_Att}}
 {id=Branch_0, children [Mixer_6,null,null], properties {boolParameterName=IsInWater}}
 {id=Doppler_0, children [Mixer_1], properties {bUseSmoothing=true}}
 {id=Mixer_0, children [ModulatorContinuous_0], properties {inputVolume [2.75]}}
 {id=Mixer_1, children [ModulatorContinuous_2, ModulatorContinuous_6, Attenuation_1, Mixer_5], properties {inputVolume [1,0.75,1.25,0.35]}}
 {id=Mixer_2, children [WavePlayer_6, ModulatorContinuous_8], properties {inputVolume [1,1.25]}}
 {id=Mixer_3, children [ModulatorContinuous_5, Attenuation_0], properties {inputVolume [1,1]}}
 {id=Mixer_4, children [WavePlayer_2,WavePlayer_5], properties {inputVolume [0.8,0.4]}}
 {id=Mixer_5, children [ModulatorContinuous_10, Attenuation_3], properties {inputVolume [1,1]}}
 {id=Mixer_6, children [ModulatorContinuous_11, ModulatorContinuous_12], properties {inputVolume [0.5,1]}}
 {id=ModulatorContinuous_0, children [Mixer_2], properties {volumeModulationParams {parameterName=OnGround}}}
 {
  id=ModulatorContinuous_1
  children [WavePlayer_1]
  properties {
   pitchModulationParams {default=0, maxOutput=1.15, minOutput=0.9, parameterName=TurnAmount}
   volumeModulationParams {default=0, parameterName=TurnAmount}
  }
 }
 {id=ModulatorContinuous_10, children [WavePlayer_4], properties {volumeModulationParams {default=0, maxOutput=3, parameterName=WaterSpeed}}}
 {
  id=ModulatorContinuous_11
  children [Modulator_1]
  properties {volumeModulationParams {default=0, maxInput=0.4, maxOutput=0, minOutput=1, parameterName=WaterSpeed}}
 }
 {
  id=ModulatorContinuous_12
  children [ModulatorContinuous_11]
  properties {
   pitchModulationParams {default=0, maxOutput=1.1, minOutput=0.9, parameterName=TurnAmount}
   volumeModulationParams {default=0, maxInput=0.5, parameterName=TurnAmount}
  }
 }
 {id=ModulatorContinuous_2, children [Mixer_0], properties {volumeModulationParams {default=0, parameterName=Speed}}}
 {
  id=ModulatorContinuous_3
  children [ModulatorContinuous_4]
  properties {pitchModulationParams {default=0, maxOutput=1.2, minOutput=0.9}, volumeModulationParams {default=0, minOutput=0.2, parameterName=InvSpeed}}
 }
 {id=ModulatorContinuous_4, children [Mixer_4], properties {volumeModulationParams {default=0, parameterName=HasDriver}}}
 {
  id=ModulatorContinuous_5
  children [ModulatorContinuous_3]
  properties {pitchModulationParams {default=0, maxOutput=1.2, minOutput=0.9, parameterName=Speed}}
 }
 {
  id=ModulatorContinuous_6
  children [ModulatorContinuous_1]
  properties {pitchModulationParams {minOutput=1, parameterName=0}, volumeModulationParams {maxOutput=1.25, minOutput=1, parameterName=Speed}}
 }
 {id=ModulatorContinuous_7, children [Mixer_3], properties {volumeModulationParams {parameterName=OutOfFuel}}}
 {id=ModulatorContinuous_8, children [WavePlayer_3], properties {volumeModulationParams {default=0, maxInput=0.5, parameterName=Damage}}}
 {
  id=ModulatorContinuous_9
  children [ModulatorContinuous_5]
  properties {
   pitchModulationParams {default=0, maxOutput=1.1, minOutput=0.9, parameterName=TurnAmount}
   volumeModulationParams {default=0, parameterName=TurnAmount}
  }
 }
 {id=Modulator_0, children [ModulatorContinuous_7], properties {pitchMax=1.2, pitchMin=1.2, volumeMax=1, volumeMin=1}}
 {id=Modulator_1, children [WavePlayer_8], properties {pitchMax=1.75, pitchMin=1.65, volumeMax=1, volumeMin=1}}
 {id=WavePlayer_1, properties {bLooping=true, soundWaveAssetPtr=&/Hamster_Turn_Loop_01}}
 {id=WavePlayer_2, properties {bLooping=true, soundWaveAssetPtr=&/Hamster_IdleLoop_01}}
 {id=WavePlayer_3, properties {bLooping=true, soundWaveAssetPtr=&/Hamster_Surface_Roll_Damaged_Loop_01}}
 {
  id=WavePlayer_4
  properties {bLooping=true, soundWaveAssetPtr=/Game/Athena/Sounds/Vehicles/Meatball/SoundWaves/Veh_Meatball_Wake_Choppy_Loop_01}
 }
 {id=WavePlayer_5, properties {bLooping=true, soundWaveAssetPtr=&/HamsterV2_Idle_Loop_01}}
 {id=WavePlayer_6, properties {bLooping=true, soundWaveAssetPtr=&/Hamster_Surface_Roll_Loop_01}}
 {
  id=WavePlayer_8
  properties {bLooping=true, soundWaveAssetPtr=/Game/Athena/Sounds/Vehicles/Meatball/SoundWaves/Veh_Meatball_Wake_Calm_Loop_01}
 }
]
properties {
 attenuationSettings=/Game/Sounds/Attenuation/Fort_GolfCart_Movement_Distant_Att
 bOverrideConcurrency=true
 maxDistance=10100
 soundClassObject=&Fort_Vehicles
 volumeMultiplier=0.7
 concurrencyOverrides {maxCount=3, resolutionRule=StopQuietest}
}
