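# Sound cue for the Hamster vehicle's water-exit movement sound in Athena.
# A Random node picks between two water exit wave variants with equal weight.
# It plays through the vehicles sound class at reduced volume, reusing the close-range golf cart movement attenuation asset.
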SoundCue /Game/Athena/Sounds/Vehicles/Hamster/SC_Hamster_Movement_Exit_Water_Cue
root=Random_1
nodes [
 {id=Random_1, children [WavePlayer_0,WavePlayer_3], properties {weights [1,1]}}
 {id=WavePlayer_0, properties {soundWaveAssetPtr=&/Hamsterv2_Water_Exit_01}}
 {id=WavePlayer_3, properties {soundWaveAssetPtr=&/Hamsterv2_Water_Exit_02}}
]
properties {
 attenuationSettings=/Game/Sounds/Attenuation/Fort_GolfCart_Movement_Close_Att
 maxDistance=5300
 soundClassObject=&Fort_Vehicles
 volumeMultiplier=0.2
}
