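# Sound cue for the Hamster vehicle starting to move, routed to the vehicle sound class with golf cart close-movement attenuation.
# A mixer layers a modulated move-start wave, a pitched-down random pick of three HamsterV2 starts, and a servo accent.
# The servo layer switches by quality level between four and two random variants, with two slots empty.
# An outer modulator wraps the mix, and concurrency allows one instance, rejecting new ones.
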SoundCue /Game/Athena/Sounds/Vehicles/Hamster/SC_Hamster_Movement_Start_Cue
root=Modulator_0
nodes [
 {id=Mixer_0, children [Modulator_2, Modulator_1, QualityLevel_0], properties {inputVolume [0.5,0.2,1]}}
 {id=Modulator_0, children [Mixer_0]}
 {id=Modulator_1, children [Random_2], properties {pitchMax=0.6, pitchMin=0.4, volumeMax=1, volumeMin=0.4}}
 {id=Modulator_2, children [WavePlayer_1], properties {pitchMax=1.15, volumeMin=0.4}}
 {id=QualityLevel_0, children [Random_1,Random_0,null,null]}
 {id=Random_0, children [WavePlayer_0,WavePlayer_2], properties {weights [1,1]}}
 {id=Random_1, children [WavePlayer_4,WavePlayer_5,WavePlayer_6,WavePlayer_7], properties {weights [1,1,1,1]}}
 {id=Random_2, children [WavePlayer_3,WavePlayer_8,WavePlayer_9], properties {weights [1,1,1]}}
 {id=WavePlayer_0, properties {soundWaveAssetPtr=&/Hamster_Servo_Small_03}}
 {id=WavePlayer_1, properties {soundWaveAssetPtr=&/Hamster_Move_Start_01}}
 {id=WavePlayer_2, properties {soundWaveAssetPtr=&/Hamster_Servo_Small_04}}
 {id=WavePlayer_3, properties {soundWaveAssetPtr=&/HamsterV2_Movement_Start_01}}
 {id=WavePlayer_4, properties {soundWaveAssetPtr=&Hamster_Servo_Small_03}}
 {id=WavePlayer_5, properties {soundWaveAssetPtr=&Hamster_Servo_Small_04}}
 {id=WavePlayer_6, properties {soundWaveAssetPtr=&/Hamster_Servo_Small_01}}
 {id=WavePlayer_7, properties {soundWaveAssetPtr=&/Hamster_Servo_Small_02}}
 {id=WavePlayer_8, properties {soundWaveAssetPtr=&/HamsterV2_Movement_Start_02}}
 {id=WavePlayer_9, properties {soundWaveAssetPtr=&/HamsterV2_Movement_Start_03}}
]
properties {
 attenuationSettings=/Game/Sounds/Attenuation/Fort_GolfCart_Movement_Close_Att
 bOverrideConcurrency=true
 maxDistance=5300
 soundClassObject=&Fort_Vehicles
 volumeMultiplier=0.5
 concurrencyOverrides {maxCount=1, resolutionRule=PreventNew}
}
