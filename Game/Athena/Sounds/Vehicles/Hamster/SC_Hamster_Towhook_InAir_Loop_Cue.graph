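# Sound cue for the looping in-air sound of the Hamster vehicle's towhook projectile.
# A mixer layers two reused grenade in-air loops, with the building grenade layer quieter and pitched up by its own fixed modulator.
# A root modulator raises the pitch of the whole mix and slightly randomizes pitch and volume.
# It uses towhook attenuation, the ranged projectile effects sound class and a tight concurrency limit.
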SoundCue /Game/Athena/Sounds/Vehicles/Hamster/SC_Hamster_Towhook_InAir_Loop_Cue
root=Modulator_1
nodes [
 {id=Mixer_3, children [WavePlayer_10, Modulator_2], properties {inputVolume [1,0.65]}}
 {id=Modulator_1, children [Mixer_3], properties {pitchMax=2, pitchMin=1.85, volumeMax=1.1, volumeMin=0.9}}
 {id=Modulator_2, children [WavePlayer_8], properties {pitchMax=1.5, pitchMin=1.5}}
 {id=WavePlayer_10, properties {bLooping=true, soundWaveAssetPtr=/Game/Sounds/Fort_Weapons/Grenades/Chiller/ChillGrenade_InAirLoop_01}}
 {id=WavePlayer_8, properties {bLooping=true, soundWaveAssetPtr=/Game/Sounds/Creative/Gadgets/Building/Building_Grenade_InAirLoop01}}
]
properties {
 attenuationSettings=&/Towhook_Projectile_InAirLoop_Att
 bOverrideConcurrency=true
 maxDistance=4000
 soundClassObject=&Fort_WeaponRanged_Projectile_Effects
 volumeMultiplier=0.5
 attenuationOverrides {bAttenuate=false, stereoSpread=0}
 concurrencyOverrides {maxCount=2}
}
