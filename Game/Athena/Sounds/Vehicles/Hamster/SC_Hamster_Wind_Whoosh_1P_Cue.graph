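# First-person wind whoosh sound cue for the Hamster vehicle, played in the vehicle sound class.
# A Mixer layers a random skydive wind-enter wave with a louder, briefly delayed random grappling-hook body-swing wave.
# The swing layer passes through a Modulator at default settings for variation.
# The cue reuses parachute and grappling-hook audio and borrows the Ferret biplane interior movement attenuation.
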SoundCue /Game/Athena/Sounds/Vehicles/Hamster/SC_Hamster_Wind_Whoosh_1P_Cue
root=Mixer_0
nodes [
 {id=Delay_0, children [Random_2], properties {delayMax=0.3, delayMin=0.25}}
 {id=Mixer_0, children [Random_0,Modulator_1], properties {inputVolume [0.5,0.75]}}
 {id=Modulator_1, children [Delay_0]}
 {id=Random_0, children [WavePlayer_1,WavePlayer_2], properties {weights [1,1]}}
 {id=Random_2, children [WavePlayer_8, WavePlayer_9, WavePlayer_10], properties {weights [1,1,1]}}
 {id=WavePlayer_1, properties {soundWaveAssetPtr=/Game/Athena/Sounds/Wind_Falling/sfx_parachute_skydive_loop_enter_01}}
 {id=WavePlayer_10, properties {soundWaveAssetPtr=/Game/Sounds/Fort_Weapons/GrapplingHook/Grappling_BodySwing_03}}
 {id=WavePlayer_2, properties {soundWaveAssetPtr=/Game/Athena/Sounds/Wind_Falling/sfx_parachute_skydive_loop_enter_02}}
 {id=WavePlayer_8, properties {soundWaveAssetPtr=/Game/Sounds/Fort_Weapons/GrapplingHook/Grappling_BodySwing_01}}
 {id=WavePlayer_9, properties {soundWaveAssetPtr=/Game/Sounds/Fort_Weapons/GrapplingHook/Grappling_BodySwing_02}}
]
properties {
 attenuationSettings=/Game/Athena/Sounds/Vehicles/Ferret/Fort_Biplane_Movement_MLTI_Interior_Att
 maxDistance=9900
 soundClassObject=&Fort_Vehicles
 volumeMultiplier=1
}
